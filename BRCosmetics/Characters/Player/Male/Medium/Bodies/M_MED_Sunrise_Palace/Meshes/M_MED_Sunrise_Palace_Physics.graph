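# Physics asset for the male medium Sunrise Palace body cosmetic, covering only its left and right strap bone chains.
# Each side has a large base capsule followed by four small capsules on strap bones 1 to 4.
# Constraints link each segment to the one before it, allowing limited, equal swing on both axes with a zero twist limit.
# The two chains have no constraint between them.
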PhysicsAsset /BRCosmetics/Characters/Player/Male/Medium/Bodies/M_MED_Sunrise_Palace/Meshes/M_MED_Sunrise_Palace_Physics
bodies {
 strapLf1 [{shapeType=Capsule, length=1, radius=3, shapeName=strapLf1_capsule, center=<1 1.7 0>, rotation=<rot 90 0 20.0009>}]
 strapLf2 [{shapeType=Capsule, length=1, radius=3, shapeName=strapLf2_capsule, center=<1 1.7 -0.0001>, rotation=<rot 90 0 -10.0001>}]
 strapLf3 [{shapeType=Capsule, length=2, radius=2, shapeName=strapLf3_capsule, center=<0.9392 2.0449 -0.0002>, rotation=<rot 90 0 -10.0001>}]
 strapLf4 [{shapeType=Capsule, length=1.5, radius=2.5, shapeName=strapLf4_capsule, center=<1.1198 2.0291 -0.0002>, rotation=<rot 90 14.0363 34.0391>}]
 strapLfBase [{shapeType=Capsule, length=7.75, radius=11.75, shapeName=strapRtBase_capsule, center=<-4.2563 -10.2656 2.5839>, rotation=<rot -10 0 0>}]
 strapRt1 [{shapeType=Capsule, length=1, radius=3, shapeName=strapLf1_capsule, center=<-1 -1.7 0>, rotation=<rot 90 -14.0363 -4.0362>}]
 strapRt2 [{shapeType=Capsule, length=1, radius=3, shapeName=strapLf2_capsule, center=<-0.9571 -1.7 0>, rotation=<rot 90 0 -10.0002>}]
 strapRt3 [{shapeType=Capsule, length=2, radius=2, shapeName=strapLf3_capsule, center=<-0.9199 -2.1544 -0.0001>, rotation=<rot 90 0 -10.0001>}]
 strapRt4 [{shapeType=Capsule, length=2, radius=2, shapeName=strapLf4_capsule, center=<-1.1578 -2.1336 -0.0001>, rotation=<rot 90 -90 -69.9983>}]
 strapRtBase [{shapeType=Capsule, length=7.75, radius=11.75, shapeName=strapRtBase_capsule, center=<4.2563 10.2656 -2.5839>, rotation=<rot -10 0 0>}]
}
constraints [
 {bone1Name=strapLf1, bone2Name=strapLfBase, swing1LimitDegrees=45, swing2LimitDegrees=45, twistLimitDegrees=0}
 {bone1Name=strapLf2, bone2Name=strapLf1, swing1LimitDegrees=45, swing2LimitDegrees=45, twistLimitDegrees=0}
 {bone1Name=strapLf3, bone2Name=strapLf2, swing1LimitDegrees=45, swing2LimitDegrees=45, twistLimitDegrees=0}
 {bone1Name=strapLf4, bone2Name=strapLf3, swing1LimitDegrees=45, swing2LimitDegrees=45, twistLimitDegrees=0}
 {bone1Name=strapRt1, bone2Name=strapRtBase, swing1LimitDegrees=45, swing2LimitDegrees=45, twistLimitDegrees=0}
 {bone1Name=strapRt2, bone2Name=strapRt1, swing1LimitDegrees=45, swing2LimitDegrees=45, twistLimitDegrees=0}
 {bone1Name=strapRt3, bone2Name=strapRt2, swing1LimitDegrees=45, swing2LimitDegrees=45, twistLimitDegrees=0}
 {bone1Name=strapRt4, bone2Name=strapRt3, swing1LimitDegrees=45, swing2LimitDegrees=45, twistLimitDegrees=0}
]
