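# Secondary-motion and deformation AnimBP for the Sunrise Palace male medium body cosmetic, copying pose from the attached parent mesh.
# Twist and sculpt Control Rigs run first, then constraints on both strap bases and a RigidBody sim in spine_05 space.
# Additive bone offsets then correct the straps, weighted by clavicle rotation deltas computed each update.
# A final right-strap rotation applies only in the front end.
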AnimBlueprint /BRCosmetics/Characters/Player/Male/Medium/Bodies/M_MED_Sunrise_Palace/Meshes/M_MED_Sunrise_Palace_AnimBP
target_skeleton=&/M_MED_Sunrise_Palace_Skeleton
graphs [
 AnimGraph {nodes [
   n1 = Control Rig(Source=Copy Pose From Mesh() {bUseAttachedParent=true}) {
    alphaBoolBlend {blendOption=Linear}
    controlRigAssetReference {blueprintRigClass=&/M_MED_Sunries_Palace_Deform_CtrlRig_twists_2.M_MED_Sunries_Palace_Deform_CtrlRig_twists_2_C}
   }
   n2 = Control Rig!(
    Source=@n1
    Alpha
    KneeRot_fr_l_114deg:Vector
    KneeTrans_fr_l_114deg:Vector="0.258,-3.591,0.009"
    KneeScale_fr_l_114deg:Vector="1,1,1"
    KneeScale_fr_l_130deg:Vector="1,1,1"
    KneeRot_fr_l_130deg:Vector="0,0,-51.2"
    KneeTrans_fr_l_130deg:Vector="0.258,-3.591,0.009"
   ) -> (Pose) {
    alphaBoolBlend {blendOption=Linear}
    controlRigAssetReference {blueprintRigClass=&/M_MED_Sunrise_Palace_Deform_CtrlRig_sculpters.M_MED_Sunrise_Palace_Deform_CtrlRig_sculpters_C}
   }
   n3 = "Constraint - strapLfBase (PPT)"!(ComponentPose=Local To Component(LocalPose=@n2), Alpha=1, ConstraintWeights_0=0.3, ConstraintWeights_1:float=1, ConstraintWeights_2:float=0) -> (Pose) {
    alphaBoolBlend {blendOption=Linear}
    boneToModify {boneName=strapLfBase}
    constraintSetup [
     {targetBone {boneName=spine_05}}
     {targetBone {boneName=clavicle_l}}
     {transformType=Translation, targetBone {boneName=deform_pectoral_fr_l}}
    ]
    constraintWeights [1,1,1]
   }
   n4 = "Constraint - strapRtBase (PPT)"!(ComponentPose=@n3, Alpha=1, ConstraintWeights_0=0.3, ConstraintWeights_1:float=1, ConstraintWeights_2:float=0.4) -> (Pose) {
    alphaBoolBlend {blendOption=Linear}
    boneToModify {boneName=strapRtBase}
    constraintSetup [
     {targetBone {boneName=spine_05}}
     {targetBone {boneName=clavicle_r}}
     {transformType=Translation, targetBone {boneName=deform_pectoral_fr_r}}
    ]
    constraintWeights [1,1,1]
   }
   n5 = RigidBody(ComponentPose=@n4) {
    bClampLinearTranslationLimitToRefPose=true
    overridePhysicsAsset=&/M_MED_Sunrise_Palace_Physics
    simulationSpace=BaseBoneSpace
    alphaBoolBlend {blendOption=Linear}
    baseBoneRef {boneName=spine_05}
    componentAppliedLinearAccClamp=<3000 3000 3000>
    componentLinearAccScale=<0.4 0.3 0.3>
    componentLinearVelScale=<0.4 0.3 0.3>
   }
   n6 = "Transform (Modify) Bone - Bone: strapLf2"(ComponentPose=@n5, Alpha=Get clavicleLRot() -> (clavicleLRot:double), Translation="2.5,-5.5,0", Rotation="0,0,0", Scale="1,1,1") {
    rotationMode=BMM_Additive
    rotationSpace=BCS_BoneSpace
    translationMode=BMM_Additive
    translationSpace=BCS_BoneSpace
    alphaBoolBlend {blendOption=Linear}
    alphaScaleBiasClamp {bInterpResult=true, interpSpeedDecreasing=100, interpSpeedIncreasing=100}
    boneToModify {boneName=strapLf2}
    translation=<2.5 -5.5>
   }
   n7 = "Transform (Modify) Bone - Bone: strapRt2"(ComponentPose=@n6, Alpha=Get clavicleRRot() -> (clavicleRRot:double), Translation="3,-3,0", Rotation="0,-10,0", Scale="1,1,1") {
    rotationMode=BMM_Additive
    rotationSpace=BCS_BoneSpace
    translationMode=BMM_Additive
    translationSpace=BCS_BoneSpace
    alphaBoolBlend {blendOption=Linear}
    alphaScaleBiasClamp {bInterpResult=true, interpSpeedDecreasing=100, interpSpeedIncreasing=100}
    boneToModify {boneName=strapRt2}
    rotation {yaw=-10}
    translation=<3 -3>
   }
   n8 = "Transform (Modify) Bone - Bone: strapRt2"(ComponentPose=@n7, Alpha=Get isFrontEnd() -> (isFrontEnd:double), Translation="0,0,0", Rotation="0,3,0", Scale="1,1,1") {
    rotationMode=BMM_Additive
    rotationSpace=BCS_BoneSpace
    translationMode=BMM_Additive
    translationSpace=BCS_BoneSpace
    alphaBoolBlend {blendOption=Linear}
    alphaScaleBiasClamp {bInterpResult=true, interpSpeedDecreasing=100, interpSpeedIncreasing=100}
    boneToModify {boneName=strapRt2}
    rotation {yaw=3}
   }
   Output Pose(Result=Component To Local(ComponentPose=@n8)) {layerGroup=None}
  ]}
 EventGraph {nodes [
   Event BlueprintUpdateAnimation()
   n1 = GetOwnerSkeletalMeshComponent()
   > Set clavicleLRot(clavicleLRot:double=GetAlphaFromReferenceTransformDelta(SkeletalMeshComponent=@n1, BoneName=clavicle_l, BaseName=spine_05, DeltaComponentType=EulerY, InputRange=(X=20,Y=100), TwistAxis="1,0,0")) -> (Output_Get:double)
   > Set clavicleRRot(clavicleRRot:double=GetAlphaFromReferenceTransformDelta(SkeletalMeshComponent=@n1, BoneName=clavicle_r, BaseName=spine_05, InputRange=(X=20,Y=100), TwistAxis="1,0,0")) -> (Output_Get:double)
   > Set isFrontEnd(isFrontEnd:double="To Float (Boolean)"(InBool=Get bIsInFrontEnd())) -> (Output_Get:double)
  ]}
]
variables [
 clavicleLRot {type=number}
 clavicleRRot {type=number}
 isFrontEnd {type=number}
]
