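# Physics asset for the face accessory part of the Sunrise Palace medium male character body, covering its dynamic hoodie bones.
# Capsule bodies sit on head, spine_05 and two hoodie bones, with the head's zero-length capsule acting as a sphere.
# All three constraints join directly to spine_05 rather than chaining: the head has free swing and twist, while both hoodie bones carry swing limits and a zero twist limit.
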PhysicsAsset /BRCosmetics/Characters/Player/Male/Medium/Bodies/M_MED_Sunrise_Palace/Meshes/Parts/M_MED_Sunrise_Palace_FaceAcc_Physics
bodies {
 head [{shapeType=Capsule, length=0, radius=13, shapeName=head_capsule, center=<14.0399 2.5734 -0.5471>}]
 hoodie_dyn_BaseA [{shapeType=Capsule, length=15.25, radius=4, shapeName=hoodie_dyn_BaseA_capsule, center=<12.9379 0.0001 0>, rotation=<rot 90 0 90.0014>}]
 hoodie_dyn_BaseB [{
   shapeType=Capsule
   length=5.25
   radius=2.75
   shapeName=hoodie_dyn_BaseB_capsule
   center=<12.9379 0.0001 0>
   rotation=<rot 90 -0.0198 -0.0198>
  }]
 spine_05 [{shapeType=Capsule, length=25, radius=9.25, shapeName=spine_05_capsule, center=<0 0.3292 0>, rotation=<rot 0 -90 0>}]
}
constraints [
 {
  bone1Name=head
  bone2Name=spine_05
  swing1LimitDegrees=45
  swing1Motion=Free
  swing2LimitDegrees=45
  swing2Motion=Free
  twistLimitDegrees=45
  twistMotion=Free
 }
 {bone1Name=hoodie_dyn_BaseB, bone2Name=spine_05, swing1LimitDegrees=45, swing2LimitDegrees=45, twistLimitDegrees=0}
 {bone1Name=hoodie_dyn_BaseA, bone2Name=spine_05, swing1LimitDegrees=45, swing2LimitDegrees=45, twistLimitDegrees=0}
]
