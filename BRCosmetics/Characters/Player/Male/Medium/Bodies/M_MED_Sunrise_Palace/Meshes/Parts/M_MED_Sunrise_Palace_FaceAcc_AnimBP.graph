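# Post-process style AnimBP for the Sunrise Palace male face accessory, copying the attached parent's pose and adding mask and hair physics.
# Two chained RigidBody nodes simulate relative to spine_05 and head, then additive bone offsets correct the mask.
# EventGraph derives offset alphas from clavicle and head deltas against spine_05, zeroing head-twist correction in the front end.
# The targeting-driven offset is zero, and the right clavicle alpha is unused.
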AnimBlueprint /BRCosmetics/Characters/Player/Male/Medium/Bodies/M_MED_Sunrise_Palace/Meshes/Parts/M_MED_Sunrise_Palace_FaceAcc_AnimBP
target_skeleton=&/M_MED_Sunrise_Palace_FaceAcc_Skeleton
graphs [
 AnimGraph {nodes [
   n1 = RigidBody(ComponentPose=Local To Component(LocalPose=Copy Pose From Mesh() {bUseAttachedParent=true})) {
    bClampLinearTranslationLimitToRefPose=true
    bOverrideWorldGravity=true
    overridePhysicsAsset=&/M_MED_Sunrise_Palace_FaceAcc_Physics
    simulationSpace=BaseBoneSpace
    alphaBoolBlend {blendOption=Linear}
    baseBoneRef {boneName=spine_05}
    componentAppliedLinearAccClamp=<3000 3000 3000>
    componentLinearAccScale=<0.4 0.3 0.3>
    componentLinearVelScale=<0.4 0.3 0.3>
    overrideWorldGravity {z=-980}
   }
   n2 = RigidBody(ComponentPose=@n1) {
    bClampLinearTranslationLimitToRefPose=true
    bOverrideWorldGravity=true
    overridePhysicsAsset=&/M_MED_Sunrise_Palace_FaceAcc_PhysicsHair
    simulationSpace=BaseBoneSpace
    alphaBoolBlend {blendOption=Linear}
    baseBoneRef {boneName=head}
    componentAppliedLinearAccClamp=<3000 3000 3000>
    componentLinearAccScale=<0.4 0.3 0.3>
    componentLinearVelScale=<0.4 0.3 0.3>
    overrideWorldGravity {z=-980}
   }
   n3 = "Transform (Modify) Bone - Bone: maskLf_dyn"(ComponentPose=@n2, Alpha=Get clavicleLRot() -> (clavicleLRot:double), Translation="0,0,6", Rotation="0,0,0", Scale="1,1,1") {
    rotationMode=BMM_Additive
    rotationSpace=BCS_BoneSpace
    translationMode=BMM_Additive
    translationSpace=BCS_BoneSpace
    alphaBoolBlend {blendOption=Linear}
    alphaScaleBiasClamp {bInterpResult=true, interpSpeedDecreasing=100, interpSpeedIncreasing=100}
    boneToModify {boneName=maskLf_dyn}
    translation {z=6}
   }
   n4 = "Transform (Modify) Bone - Bone: mask_dyn"(ComponentPose=@n3, Alpha="To Float (Boolean)"(InBool=Get bIsTargeting()), Translation="0,0,0", Rotation="0,0,0", Scale="1,1,1") {
    rotationMode=BMM_Additive
    rotationSpace=BCS_BoneSpace
    translationMode=BMM_Additive
    translationSpace=BCS_BoneSpace
    alphaBoolBlend {blendOption=Linear}
    alphaScaleBiasClamp {bInterpResult=true, interpSpeedDecreasing=100, interpSpeedIncreasing=100}
    boneToModify {boneName=mask_dyn}
   }
   n5 = "Transform (Modify) Bone - Bone: mask_dyn"(ComponentPose=@n4, Alpha=Get headDnLfTwist() -> (headDnLfTwist:double), Translation="0,-40,0", Rotation="0,250,0", Scale="1,1,1") {
    rotationMode=BMM_Additive
    rotationSpace=BCS_BoneSpace
    translationMode=BMM_Additive
    translationSpace=BCS_BoneSpace
    alphaBoolBlend {blendOption=Linear}
    alphaScaleBiasClamp {bInterpResult=true, interpSpeedDecreasing=100, interpSpeedIncreasing=100}
    boneToModify {boneName=mask_dyn}
    rotation {yaw=250}
    translation {y=-40}
   }
   n6 = "Transform (Modify) Bone - Bone: mask_dyn"(ComponentPose=@n5, Alpha=Get headDnRtTwist() -> (headDnRtTwist:double), Translation="0,40,0", Rotation="0,-250,0", Scale="1,1,1") {
    rotationMode=BMM_Additive
    rotationSpace=BCS_BoneSpace
    translationMode=BMM_Additive
    translationSpace=BCS_BoneSpace
    alphaBoolBlend {blendOption=Linear}
    alphaScaleBiasClamp {bInterpResult=true, interpSpeedDecreasing=100, interpSpeedIncreasing=100}
    boneToModify {boneName=mask_dyn}
    rotation {yaw=-250}
    translation {y=40}
   }
   Output Pose(Result=Component To Local(ComponentPose=@n6)) {layerGroup=None}
  ]}
 EventGraph {nodes [
   Event BlueprintUpdateAnimation()
   > n1 = Sequence()
   > n2 = Branch(Condition=Get bIsInFrontEnd())
   > Set Is in Front End(Is in Front End:double=0) -> (Output_Get:double)
   Set Is in Front End(execute=@n2.else, Is in Front End=1)
   n3 = GetOwnerSkeletalMeshComponent()
   Set clavicleLRot(
    execute=@n1.then_1
    clavicleLRot:double=GetAlphaFromReferenceTransformDelta(SkeletalMeshComponent=@n3, BoneName=clavicle_l, BaseName=spine_05, DeltaComponentType=EulerY, InputRange=(X=20,Y=100), TwistAxis="1,0,0")
   ) -> (Output_Get:double)
   > Set clavicleRRot(clavicleRRot:double=GetAlphaFromReferenceTransformDelta(SkeletalMeshComponent=@n3, BoneName=clavicle_r, BaseName=spine_05, InputRange=(X=20,Y=100), TwistAxis="1,0,0")) -> (Output_Get:double)
   n4 = GetOwnerSkeletalMeshComponent()
   > n5 = Set HeadFwd(HeadFwd:double=GetAlphaFromReferenceTransformDelta(SkeletalMeshComponent=@n4, BoneName=head, BaseName=spine_05, InputRange=(X=5,Y=90), TwistAxis="1,0,0")) -> (Output_Get:double)
   > n6 = Set HeadLfTwist(HeadLfTwist:double=GetAlphaFromReferenceTransformDelta(SkeletalMeshComponent=@n4, BoneName=head, BaseName=spine_05, DeltaComponentType=EulerX, InputRange=(X=3,Y=50), TwistAxis="1,0,0")) -> (Output_Get:double)
   > n7 = Set HeadRtTwist(HeadRtTwist:double=GetAlphaFromReferenceTransformDelta(SkeletalMeshComponent=@n4, BoneName=head, BaseName=spine_05, DeltaComponentType=EulerX, InputRange=(X=3,Y=-50), TwistAxis="1,0,0")) -> (Output_Get:double)
   n8 = Get Is in Front End() -> (Is in Front End:double)
   > Set headDnLfTwist(headDnLfTwist:double=float * float!(A=float * float!(A=1.5, B=@n5) -> (ReturnValue), B=@n6, C:double=@n8) -> (ReturnValue)) -> (Output_Get:double)
   > Set headDnRtTwist(headDnRtTwist:double=float * float(A=@n7, B=float * float!(A=1.5, B=@n5, C:double=@n8) -> (ReturnValue))) -> (Output_Get:double)
  ]}
]
variables [
 clavicleLRot {type=number}
 clavicleRRot {type=number}
 Head Rotation {type=number}
 headDnLfTwist {type=number}
 headDnRtTwist {type=number}
 HeadFwd {type=number}
 HeadLfTwist {type=number}
 headRot {type=number}
 HeadRtTwist {type=number}
 Is in Front End {type=number}
 Is Targeting {type=number}
]
