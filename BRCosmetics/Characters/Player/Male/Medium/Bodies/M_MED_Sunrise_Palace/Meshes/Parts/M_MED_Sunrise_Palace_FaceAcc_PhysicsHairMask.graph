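# Physics asset for the Sunrise Palace male medium body's face accessory, setting up dynamic hair on the PhysicsHairMask part.
# A head capsule anchors center, left and right hair chains, each bone carrying a small uniform capsule.
# Constraints link the chain roots to the head and each later bone to one earlier in its chain, with branches in several chains.
# Every joint allows limited swing and specifies zero twist degrees.
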PhysicsAsset /BRCosmetics/Characters/Player/Male/Medium/Bodies/M_MED_Sunrise_Palace/Meshes/Parts/M_MED_Sunrise_Palace_FaceAcc_PhysicsHairMask
bodies {
 hair_dyn_Mid1 [{shapeType=Capsule, length=1, radius=1, shapeName=hair_dyn_Mid1_capsule, center=<1.0904 0 -0.25>, rotation=<rot -69.9998 0.0001 0>}]
 hair_dyn_Mid2 [{shapeType=Capsule, length=1, radius=1, shapeName=hair_dyn_Mid2_capsule, center=<0.9917 0 0.3>, rotation=<rot -79.9954 -179.9999 180>}]
 hair_dyn_Mid3 [{shapeType=Capsule, length=1, radius=1, shapeName=hair_dyn_Mid3_capsule, center=<1 0 0>, rotation=<rot -90 -179.9802 179.9803>}]
 hair_dyn_MidALf1 [{shapeType=Capsule, length=1, radius=1, shapeName=hair_dyn_MidALf1_capsule, center=<1 0 0>, rotation=<rot -60 0 0>}]
 hair_dyn_MidALf2 [{shapeType=Capsule, length=1, radius=1, shapeName=hair_dyn_MidALf2_capsule, center=<1 0 1>, rotation=<rot -79.998 180 180>}]
 hair_dyn_MidALf3 [{shapeType=Capsule, length=1, radius=1, shapeName=hair_dyn_MidALf3_capsule, center=<1 0 0>, rotation=<rot -79.9943 180 180>}]
 hair_dyn_MidARt1 [{shapeType=Capsule, length=1, radius=1, shapeName=hair_dyn_MidARt1_capsule, center=<-0.866 0 -0.1818>, rotation=<rot -59.9999 0 0>}]
 hair_dyn_MidARt2 [{shapeType=Capsule, length=1, radius=1, shapeName=hair_dyn_MidARt2_capsule, center=<-1 0 -1>, rotation=<rot -79.9966 180 -180>}]
 hair_dyn_MidARt3 [{shapeType=Capsule, length=1, radius=1, shapeName=hair_dyn_MidARt3_capsule, center=<-1 0 0>, rotation=<rot -90 -0.0297 0.0297>}]
 hair_dyn_MidBLf1 [{shapeType=Capsule, length=1, radius=1, shapeName=hair_dyn_MidBLf1_capsule, center=<0.9885 0 1.2>, rotation=<rot -69.9998 0.0001 0>}]
 hair_dyn_MidBLf2 [{shapeType=Capsule, length=1, radius=1, shapeName=hair_dyn_MidBLf2_capsule, center=<1.3689 0 0.6>, rotation=<rot -59.9973 -179.9999 180>}]
 hair_dyn_MidBLf3 [{shapeType=Capsule, length=1, radius=1, shapeName=hair_dyn_MidBLf3_capsule, center=<1 0 0>, rotation=<rot -40.0009 -179.9998 179.9999>}]
 hair_dyn_MidBRt1 [{shapeType=Capsule, length=1, radius=1, shapeName=hair_dyn_MidBRt1_capsule, center=<-1.0395 0 -0.7>, rotation=<rot -69.9997 0.0001 0>}]
 hair_dyn_MidBRt2 [{
   shapeType=Capsule
   length=1
   radius=1
   shapeName=hair_dyn_MidBRt2_capsule
   center=<-1.4156 0 -0.7>
   rotation=<rot -59.9996 179.9998 169.9999>
  }]
 hair_dyn_MidBRt3 [{shapeType=Capsule, length=1, radius=1, shapeName=hair_dyn_MidBRt3_capsule, center=<-0.9499 0 0.16>, rotation=<rot -69.995 149.9958 169.9999>}]
 hair_dyn_MidBRt4 [{shapeType=Capsule, length=1, radius=1, shapeName=hair_dyn_MidBRt4_capsule, center=<0.2341 0 0.9>, rotation=<rot -50.0003 -179.9999 179.9999>}]
 hair_dyn_MidCLf1 [{shapeType=Capsule, length=1, radius=1, shapeName=hair_dyn_MidCLf1_capsule, center=<1 0 0>, rotation=<rot -90 0 0>}]
 hair_dyn_MidCLf2 [{shapeType=Capsule, length=1, radius=1, shapeName=hair_dyn_MidCLf2_capsule, center=<1 0 0>, rotation=<rot -90 -1.8984 1.8984>}]
 hair_dyn_MidCLf3 [{shapeType=Capsule, length=1, radius=1, shapeName=hair_dyn_MidCLf3_capsule, center=<1 0 0>, rotation=<rot -90 174.9489 -174.9488>}]
 hair_dyn_MidCLf4 [{shapeType=Capsule, length=1, radius=1, shapeName=hair_dyn_MidCLf4_capsule, center=<1 0 0>, rotation=<rot -90 174.9489 -174.9488>}]
 hair_dyn_MidCLf5 [{shapeType=Capsule, length=1, radius=1, shapeName=hair_dyn_MidCLf5_capsule, center=<1 0 0>, rotation=<rot -90 5.0512 -5.0512>}]
 hair_dyn_MidCLf6 [{shapeType=Capsule, length=1, radius=1, shapeName=hair_dyn_MidCLf6_capsule, center=<1 0 0>, rotation=<rot -90 1.2659 -1.2659>}]
 hair_dyn_MidCLf7 [{shapeType=Capsule, length=1, radius=1, shapeName=hair_dyn_MidCLf7_capsule, center=<1 0 0>, rotation=<rot -90 2.5305 -2.5305>}]
 hair_dyn_MidCLf8 [{shapeType=Capsule, length=1, radius=1, shapeName=hair_dyn_MidCLf8_capsule, center=<1 0 0>, rotation=<rot -90 2.5305 -2.5304>}]
 hair_dyn_MidCLf9 [{shapeType=Capsule, length=1, radius=1, shapeName=hair_dyn_MidCLf9_capsule, center=<1 0 0>, rotation=<rot -90 2.5305 -2.5304>}]
 hair_dyn_MidCRt1 [{shapeType=Capsule, length=1, radius=1, shapeName=hair_dyn_MidCRt1_capsule, center=<-1 0 0>, rotation=<rot -90 0 0>}]
 hair_dyn_MidCRt2 [{shapeType=Capsule, length=1, radius=1, shapeName=hair_dyn_MidCRt2_capsule, center=<-1 0 0>, rotation=<rot -90 0.0099 -0.0099>}]
 hair_dyn_MidCRt3 [{shapeType=Capsule, length=1, radius=1, shapeName=hair_dyn_MidCRt3_capsule, center=<-1 0 0>, rotation=<rot -90 0.0396 -0.0396>}]
 hair_dyn_MidCRt4 [{shapeType=Capsule, length=1, radius=1, shapeName=hair_dyn_MidCRt4_capsule, center=<-1 0 0>, rotation=<rot -90 -7.5524 7.5525>}]
 hair_dyn_MidCRt5 [{shapeType=Capsule, length=1, radius=1, shapeName=hair_dyn_MidCRt5_capsule, center=<-1 0 0>, rotation=<rot -90 -7.5524 7.5525>}]
 hair_dyn_MidCRt6 [{shapeType=Capsule, length=1, radius=1, shapeName=hair_dyn_MidCRt6_capsule, center=<-1 0 0>, rotation=<rot -90 -14.8511 14.8511>}]
 hair_dyn_MidCRt7 [{shapeType=Capsule, length=1, radius=1, shapeName=hair_dyn_MidCRt7_capsule, center=<-1 0 0>, rotation=<rot -90 0.0198 -0.0199>}]
 hair_dyn_MidCRt8 [{shapeType=Capsule, length=1, radius=1, shapeName=hair_dyn_MidCRt8_capsule, center=<-1 0 0>, rotation=<rot -90 0.0198 -0.0199>}]
 head [{shapeType=Capsule, length=0, radius=7.25, shapeName=head_capsule, center=<6.0873 2.3465 0>, rotation=<rot 0 0 90.0002>}]
}
constraints [
 {bone1Name=hair_dyn_Mid2, bone2Name=hair_dyn_Mid1, swing1LimitDegrees=45, swing2LimitDegrees=45, twistLimitDegrees=0}
 {bone1Name=hair_dyn_Mid3, bone2Name=hair_dyn_Mid2, swing1LimitDegrees=45, swing2LimitDegrees=45, twistLimitDegrees=0}
 {bone1Name=hair_dyn_MidALf2, bone2Name=hair_dyn_MidALf1, swing1LimitDegrees=45, swing2LimitDegrees=45, twistLimitDegrees=0}
 {bone1Name=hair_dyn_MidALf3, bone2Name=hair_dyn_MidALf2, swing1LimitDegrees=45, swing2LimitDegrees=45, twistLimitDegrees=0}
 {bone1Name=hair_dyn_MidBLf2, bone2Name=hair_dyn_MidBLf1, swing1LimitDegrees=45, swing2LimitDegrees=45, twistLimitDegrees=0}
 {bone1Name=hair_dyn_MidBLf3, bone2Name=hair_dyn_MidBLf2, swing1LimitDegrees=45, swing2LimitDegrees=45, twistLimitDegrees=0}
 {bone1Name=hair_dyn_MidCLf2, bone2Name=hair_dyn_MidCLf1, swing1LimitDegrees=45, swing2LimitDegrees=45, twistLimitDegrees=0}
 {bone1Name=hair_dyn_MidCLf3, bone2Name=hair_dyn_MidCLf2, swing1LimitDegrees=45, swing2LimitDegrees=45, twistLimitDegrees=0}
 {bone1Name=hair_dyn_MidCLf4, bone2Name=hair_dyn_MidCLf3, swing1LimitDegrees=45, swing2LimitDegrees=45, twistLimitDegrees=0}
 {bone1Name=hair_dyn_MidCLf5, bone2Name=hair_dyn_MidCLf1, swing1LimitDegrees=45, swing2LimitDegrees=45, twistLimitDegrees=0}
 {bone1Name=hair_dyn_MidCLf6, bone2Name=hair_dyn_MidCLf5, swing1LimitDegrees=45, swing2LimitDegrees=45, twistLimitDegrees=0}
 {bone1Name=hair_dyn_MidCLf7, bone2Name=hair_dyn_MidCLf6, swing1LimitDegrees=45, swing2LimitDegrees=45, twistLimitDegrees=0}
 {bone1Name=hair_dyn_MidCLf8, bone2Name=hair_dyn_MidCLf5, swing1LimitDegrees=45, swing2LimitDegrees=45, twistLimitDegrees=0}
 {bone1Name=hair_dyn_MidCLf9, bone2Name=hair_dyn_MidCLf5, swing1LimitDegrees=45, swing2LimitDegrees=45, twistLimitDegrees=0}
 {bone1Name=hair_dyn_MidCRt2, bone2Name=hair_dyn_MidCRt1, swing1LimitDegrees=45, swing2LimitDegrees=45, twistLimitDegrees=0}
 {bone1Name=hair_dyn_MidCRt3, bone2Name=hair_dyn_MidCRt2, swing1LimitDegrees=45, swing2LimitDegrees=45, twistLimitDegrees=0}
 {bone1Name=hair_dyn_MidCRt4, bone2Name=hair_dyn_MidCRt3, swing1LimitDegrees=45, swing2LimitDegrees=45, twistLimitDegrees=0}
 {bone1Name=hair_dyn_MidCRt5, bone2Name=hair_dyn_MidCRt4, swing1LimitDegrees=45, swing2LimitDegrees=45, twistLimitDegrees=0}
 {bone1Name=hair_dyn_MidCRt8, bone2Name=hair_dyn_MidCRt3, swing1LimitDegrees=45, swing2LimitDegrees=45, twistLimitDegrees=0}
 {bone1Name=hair_dyn_MidCRt6, bone2Name=hair_dyn_MidCRt2, swing1LimitDegrees=45, swing2LimitDegrees=45, twistLimitDegrees=0}
 {bone1Name=hair_dyn_MidCRt7, bone2Name=hair_dyn_MidCRt6, swing1LimitDegrees=45, swing2LimitDegrees=45, twistLimitDegrees=0}
 {bone1Name=hair_dyn_MidBRt2, bone2Name=hair_dyn_MidBRt1, swing1LimitDegrees=45, swing2LimitDegrees=45, twistLimitDegrees=0}
 {bone1Name=hair_dyn_MidBRt3, bone2Name=hair_dyn_MidBRt2, swing1LimitDegrees=45, swing2LimitDegrees=45, twistLimitDegrees=0}
 {bone1Name=hair_dyn_MidBRt4, bone2Name=hair_dyn_MidBRt2, swing1LimitDegrees=45, swing2LimitDegrees=45, twistLimitDegrees=0}
 {bone1Name=hair_dyn_MidARt3, bone2Name=hair_dyn_MidARt2, swing1LimitDegrees=45, swing2LimitDegrees=45, twistLimitDegrees=0}
 {bone1Name=hair_dyn_MidARt2, bone2Name=hair_dyn_MidARt1, swing1LimitDegrees=45, swing2LimitDegrees=45, twistLimitDegrees=0}
 {bone1Name=hair_dyn_Mid1, bone2Name=head, swing1LimitDegrees=45, swing2LimitDegrees=45, twistLimitDegrees=0}
 {bone1Name=hair_dyn_MidALf1, bone2Name=head, swing1LimitDegrees=45, swing2LimitDegrees=45, twistLimitDegrees=0}
 {bone1Name=hair_dyn_MidBLf1, bone2Name=head, swing1LimitDegrees=45, swing2LimitDegrees=45, twistLimitDegrees=0}
 {bone1Name=hair_dyn_MidCLf1, bone2Name=head, swing1LimitDegrees=45, swing2LimitDegrees=45, twistLimitDegrees=0}
 {bone1Name=hair_dyn_MidCRt1, bone2Name=head, swing1LimitDegrees=45, swing2LimitDegrees=45, twistLimitDegrees=0}
 {bone1Name=hair_dyn_MidBRt1, bone2Name=head, swing1LimitDegrees=45, swing2LimitDegrees=45, twistLimitDegrees=0}
 {bone1Name=hair_dyn_MidARt1, bone2Name=head, swing1LimitDegrees=45, swing2LimitDegrees=45, twistLimitDegrees=0}
]
