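# Animation Blueprint for the Sunrise Palace male face accessory mask, adding physics-driven secondary motion on top of the pose copied from the attached parent mesh.
# Two chained RigidBody nodes simulate separate physics assets, one relative to the head, the other to spine_05.
# An additive bone offset then shifts maskLf_dyn, weighted by left clavicle rotation relative to spine_05.
# The event graph also computes a right clavicle alpha, which the AnimGraph leaves unused.
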AnimBlueprint /BRCosmetics/Characters/Player/Male/Medium/Bodies/M_MED_Sunrise_Palace/Meshes/Parts/M_MED_Sunrise_Palace_Mask_FaceAcc_AnimBP
target_skeleton=&/M_MED_Sunrise_Palace_Mask_FaceAcc_Skeleton
graphs [
 AnimGraph {nodes [
   n1 = RigidBody(ComponentPose=Local To Component(LocalPose=Copy Pose From Mesh() {bUseAttachedParent=true})) {
    bClampLinearTranslationLimitToRefPose=true
    bOverrideWorldGravity=true
    overridePhysicsAsset=&/M_MED_Sunrise_Palace_FaceAcc_PhysicsHairMask
    simulationSpace=BaseBoneSpace
    alphaBoolBlend {blendOption=Linear}
    baseBoneRef {boneName=head}
    componentAppliedLinearAccClamp=<3000 3000 3000>
    componentLinearAccScale=<0.4 0.3 0.3>
    componentLinearVelScale=<0.4 0.3 0.3>
    overrideWorldGravity {z=-980}
   }
   n2 = RigidBody(ComponentPose=@n1) {
    bClampLinearTranslationLimitToRefPose=true
    bOverrideWorldGravity=true
    overridePhysicsAsset=&/M_MED_Sunrise_Palace_FaceAcc_PhysicsMask
    simulationSpace=BaseBoneSpace
    alphaBoolBlend {blendOption=Linear}
    baseBoneRef {boneName=spine_05}
    componentAppliedLinearAccClamp=<3000 3000 3000>
    componentLinearAccScale=<0.4 0.3 0.3>
    componentLinearVelScale=<0.4 0.3 0.3>
    overrideWorldGravity {z=-980}
   }
   n3 = "Transform (Modify) Bone - Bone: maskLf_dyn"(ComponentPose=@n2, Alpha=Get clavicleLRot() -> (clavicleLRot:double), Translation="0,0,5", Rotation="0,0,0", Scale="1,1,1") {
    rotationMode=BMM_Additive
    rotationSpace=BCS_BoneSpace
    translationMode=BMM_Additive
    translationSpace=BCS_BoneSpace
    alphaBoolBlend {blendOption=Linear}
    alphaScaleBiasClamp {bInterpResult=true, interpSpeedDecreasing=100, interpSpeedIncreasing=100}
    boneToModify {boneName=maskLf_dyn}
    translation {z=5}
   }
   Output Pose(Result=Component To Local(ComponentPose=@n3)) {layerGroup=None}
  ]}
 EventGraph {nodes [
   Event BlueprintUpdateAnimation()
   n1 = GetOwnerSkeletalMeshComponent()
   > Set clavicleLRot(clavicleLRot:double=GetAlphaFromReferenceTransformDelta(SkeletalMeshComponent=@n1, BoneName=clavicle_l, BaseName=spine_05, DeltaComponentType=EulerY, InputRange=(X=20,Y=100), TwistAxis="1,0,0")) -> (Output_Get:double)
   > Set clavicleRRot(clavicleRRot:double=GetAlphaFromReferenceTransformDelta(SkeletalMeshComponent=@n1, BoneName=clavicle_r, BaseName=spine_05, InputRange=(X=20,Y=100), TwistAxis="1,0,0")) -> (Output_Get:double)
  ]}
]
variables [
 clavicleLRot {type=number}
 clavicleRRot {type=number}
]
